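# Sound cue for the Meatball vehicle's boost activation, layering a boost start sound with high-acceleration engine audio.
# A nested mixer blends exterior and engine high-accel waves, and the root mixer plays that blend at half volume against the full-volume boost start wave.
# It uses boat boost attenuation and the exterior vehicle sound class, with an owner-limited concurrency override that keeps one instance and stops the oldest.
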SoundCue /Game/Athena/Sounds/Vehicles/Meatball/Meatball_Boost_Activate_Cue
root=Mixer_1
nodes [
 {id=Mixer_0, children [WavePlayer_1,WavePlayer_2], properties {inputVolume [1,1]}}
 {id=Mixer_1, children [Mixer_0,WavePlayer_0], properties {inputVolume [0.5,1]}}
 {id=WavePlayer_0, properties {soundWaveAssetPtr=/Game/Athena/Sounds/Vehicles/Meatball/SoundWaves/Veh_Meatball_Boost_Start_01}}
 {id=WavePlayer_1, properties {soundWaveAssetPtr=/Game/Athena/Sounds/Vehicles/Meatball/SoundWaves/Veh_Meatball_High_Accel_Ext_01}}
 {id=WavePlayer_2, properties {soundWaveAssetPtr=/Game/Athena/Sounds/Vehicles/Meatball/SoundWaves/Veh_Meatball_High_Accel_Engine_01}}
]
properties {
 attenuationSettings=/Game/Sounds/Attenuation/Fort_Boat_Engine_Boost_Activate_Att
 bOverrideConcurrency=true
 maxDistance=18000
 soundClassObject=&Fort_Vehicles_Exterior
 volumeMultiplier=1.2
 concurrencyOverrides {bLimitToOwner=true, maxCount=1, resolutionRule=StopOldest, voiceStealReleaseTime=1}
}
