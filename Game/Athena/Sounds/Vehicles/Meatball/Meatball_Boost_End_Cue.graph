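# Sound cue for the end of a boost on the Meatball vehicle, playing as an exterior vehicle sound with boat engine boost attenuation.
# A top mixer layers the boost-end one-shot with a deceleration branch that starts after a short fixed delay.
# That branch mixes high-decel engine and high-decel "Ext" waves through an enveloper, which holds full volume briefly, then fades to silence over about two seconds.
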SoundCue /Game/Athena/Sounds/Vehicles/Meatball/Meatball_Boost_End_Cue
root=Mixer_0
nodes [
 {id=Delay_0, children [Enveloper_0], properties {delayMax=0.2, delayMin=0.2}}
 {
  id=Enveloper_0
  children [Mixer_1]
  properties {pitchCurve {editorCurveData {keys [{time=0}]}}, volumeCurve {editorCurveData {keys [<at 0 1>,<at 0.5 1>,<at 2 0>]}}}
 }
 {id=Mixer_0, children [Delay_0,WavePlayer_0], properties {inputVolume [1,1]}}
 {id=Mixer_1, children [WavePlayer_1,WavePlayer_2], properties {inputVolume [1,1]}}
 {id=WavePlayer_0, properties {soundWaveAssetPtr=/Game/Athena/Sounds/Vehicles/Meatball/SoundWaves/Veh_Meatball_Boost_End_01}}
 {id=WavePlayer_1, properties {soundWaveAssetPtr=/Game/Athena/Sounds/Vehicles/Meatball/SoundWaves/Veh_Meatball_High_Decel_Engine_01}}
 {id=WavePlayer_2, properties {soundWaveAssetPtr=/Game/Athena/Sounds/Vehicles/Meatball/SoundWaves/Veh_Meatball_High_Decel_Ext_01}}
]
properties {
 attenuationSettings=/Game/Sounds/Attenuation/Fort_Boat_Engine_Boost_Activate_Att
 maxDistance=18000
 soundClassObject=&Fort_Vehicles_Exterior
 volumeMultiplier=0.8
}
