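# Sound cue for the boost-ready sound of the Meatball vehicle, judging by its name, using a close-detail boat engine attenuation.
# A single wave player feeds the output directly at slightly reduced volume, routed to the exterior vehicle sound class.
# An overridden concurrency rule allows one instance per owner, stopping the oldest when a new one starts.
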SoundCue /Game/Athena/Sounds/Vehicles/Meatball/Meatball_Boost_Ready_Cue
root=WavePlayer_3
nodes [{id=WavePlayer_3, properties {soundWaveAssetPtr=/Game/Athena/Sounds/Vehicles/Meatball/SoundWaves/Veh_Meatball_Boost_Ready_01}}]
properties {
 attenuationSettings=/Game/Sounds/Attenuation/Fort_Boat_Engine_Close_Detail_Att
 bOverrideConcurrency=true
 maxDistance=18000
 soundClassObject=&Fort_Vehicles_Exterior
 volumeMultiplier=0.8
 concurrencyOverrides {bLimitToOwner=true, maxCount=1, resolutionRule=StopOldest, voiceStealReleaseTime=1}
}
