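# Sound cue for the Meatball vehicle, named as the feedback for a boost that cannot be used.
# A Mixer layers a pistol foley clip, pitched up by a fixed Modulator and weighted louder, with a quieter negative UI sting.
# It plays in the UI sound class with overridden concurrency that prevents new instances once the limit is reached.
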SoundCue /Game/Athena/Sounds/Vehicles/Meatball/Meatball_Boost_Unable_Cue
root=Mixer_0
nodes [
 {id=Mixer_0, children [Modulator_0,WavePlayer_0], properties {inputVolume [2,0.4]}}
 {id=Modulator_0, children [WavePlayer_1], properties {pitchMax=1.3, pitchMin=1.3}}
 {id=WavePlayer_0, properties {soundWaveAssetPtr=/Game/Sounds/Fort_UI_Sounds/Gameplay/fort_ui_new_neg_01}}
 {id=WavePlayer_1, properties {soundWaveAssetPtr=/Game/Sounds/Fort_Weapons/pistol/pistol_foley/auto_sdk_end_01}}
]
properties {
 bOverrideConcurrency=true
 maxDistance=2097152
 soundClassObject=&Fort_UI
 volumeMultiplier=0.7
 concurrencyOverrides {resolutionRule=PreventNew}
}
