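# Sound cue for light collision impacts of the Meatball vehicle.
# A Random node picks among six equally weighted dry light impact waves, feeding a root Modulator for pitch and volume variation.
# It plays in the exterior vehicles sound class at slightly reduced volume, borrowing a close-detail boat engine attenuation asset for spatial falloff.
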SoundCue /Game/Athena/Sounds/Vehicles/Meatball/Meatball_Collision_01_Light_Cue
root=Modulator_0
nodes [
 {id=Modulator_0, children [Random_0]}
 {
  id=Random_0
  children [WavePlayer_1, WavePlayer_2, WavePlayer_3, WavePlayer_4, WavePlayer_10, WavePlayer_11]
  properties {bShouldExcludeFromBranchCulling=true, weights [1,1,1,1,1,1]}
 }
 {id=WavePlayer_1, properties {soundWaveAssetPtr=/Game/Athena/Sounds/Vehicles/Meatball/SoundWaves/Veh_Meatball_Dry_Impact_Lt_01}}
 {id=WavePlayer_10, properties {soundWaveAssetPtr=/Game/Athena/Sounds/Vehicles/Meatball/SoundWaves/Veh_Meatball_Dry_Impact_Lt_05}}
 {id=WavePlayer_11, properties {soundWaveAssetPtr=/Game/Athena/Sounds/Vehicles/Meatball/SoundWaves/Veh_Meatball_Dry_Impact_Lt_06}}
 {id=WavePlayer_2, properties {soundWaveAssetPtr=/Game/Athena/Sounds/Vehicles/Meatball/SoundWaves/Veh_Meatball_Dry_Impact_Lt_02}}
 {id=WavePlayer_3, properties {soundWaveAssetPtr=/Game/Athena/Sounds/Vehicles/Meatball/SoundWaves/Veh_Meatball_Dry_Impact_Lt_03}}
 {id=WavePlayer_4, properties {soundWaveAssetPtr=/Game/Athena/Sounds/Vehicles/Meatball/SoundWaves/Veh_Meatball_Dry_Impact_Lt_04}}
]
properties {
 attenuationSettings=/Game/Sounds/Attenuation/Fort_Boat_Engine_Close_Detail_Att
 maxDistance=18000
 soundClassObject=&Fort_Vehicles_Exterior
 volumeMultiplier=0.9
}
