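# Sound cue for heavy dry collision impacts on the Meatball vehicle.
# A Random node picks one of six equally weighted heavy impact waves and feeds it through a Modulator at default settings.
# Playback uses the exterior vehicle sound class and borrows a boat engine close-detail attenuation asset.
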SoundCue /Game/Athena/Sounds/Vehicles/Meatball/Meatball_Collision_02_Heavy_Cue
root=Modulator_0
nodes [
 {id=Modulator_0, children [Random_1]}
 {
  id=Random_1
  children [WavePlayer_0,WavePlayer_5,WavePlayer_6,WavePlayer_7,WavePlayer_8,WavePlayer_9]
  properties {bShouldExcludeFromBranchCulling=true, weights [1,1,1,1,1,1]}
 }
 {id=WavePlayer_0, properties {soundWaveAssetPtr=/Game/Athena/Sounds/Vehicles/Meatball/SoundWaves/Veh_Meatball_Dry_Impact_Hv_01}}
 {id=WavePlayer_5, properties {soundWaveAssetPtr=/Game/Athena/Sounds/Vehicles/Meatball/SoundWaves/Veh_Meatball_Dry_Impact_Hv_02}}
 {id=WavePlayer_6, properties {soundWaveAssetPtr=/Game/Athena/Sounds/Vehicles/Meatball/SoundWaves/Veh_Meatball_Dry_Impact_Hv_03}}
 {id=WavePlayer_7, properties {soundWaveAssetPtr=/Game/Athena/Sounds/Vehicles/Meatball/SoundWaves/Veh_Meatball_Dry_Impact_Hv_04}}
 {id=WavePlayer_8, properties {soundWaveAssetPtr=/Game/Athena/Sounds/Vehicles/Meatball/SoundWaves/Veh_Meatball_Dry_Impact_Hv_05}}
 {id=WavePlayer_9, properties {soundWaveAssetPtr=/Game/Athena/Sounds/Vehicles/Meatball/SoundWaves/Veh_Meatball_Dry_Impact_Hv_06}}
]
properties {
 attenuationSettings=/Game/Sounds/Attenuation/Fort_Boat_Engine_Close_Detail_Att
 maxDistance=18000
 soundClassObject=&Fort_Vehicles_Exterior
}
