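# Sound cue for the destruction explosion of the Meatball vehicle, built from reused weapon explosion and debris waves.
# After a short fixed delay, a mixer layers four branches: close Clinger blast, medium Clinger blast, distant grenade blast and metal debris.
# Each branch picks a random variant and has its own attenuation and sound class override, so the blend changes with distance.
# Modulators vary each layer, with explicit pitch ranges on medium and distant.
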SoundCue /Game/Athena/Sounds/Vehicles/Meatball/Meatball_Destroy_Cue
root=Delay_0
nodes [
 {
  id=Attenuation_0
  children [Modulator_1]
  properties {
   attenuationSettings=/Game/Sounds/Attenuation/Fort_Explosion_Distant_Athena_Att
   attenuationOverrides {falloffDistance=3000, attenuationShapeExtents {x=1000}}
  }
 }
 {
  id=Attenuation_1
  children [Random_5]
  properties {attenuationSettings=/Game/Sounds/Attenuation/Fort_Explosion_Close_Athena_Att, attenuationOverrides {falloffDistance=1600}}
 }
 {
  id=Attenuation_4
  children [Random_2]
  properties {attenuationSettings=&Fort_Explosion_Close_Athena_Att, attenuationOverrides {falloffDistance=1600}}
 }
 {
  id=Attenuation_5
  children [Modulator_5]
  properties {
   attenuationSettings=/Game/Sounds/Attenuation/Fort_Explosion_Medium_Athena_Att
   attenuationOverrides {falloffDistance=3000, attenuationShapeExtents {x=1000}}
  }
 }
 {id=Delay_0, children [Mixer_4], properties {delayMax=0.1, delayMin=0.1}}
 {id=Mixer_4, children [SoundClass_3,SoundClass_2,SoundClass_0,SoundClass_1], properties {inputVolume [1,1,1,1]}}
 {id=Modulator_0, children [Attenuation_4]}
 {id=Modulator_1, children [Random_4], properties {pitchMax=1.1, pitchMin=0.9}}
 {id=Modulator_2, children [Attenuation_1]}
 {id=Modulator_5, children [Random_3], properties {pitchMax=1.1, pitchMin=0.9}}
 {id=Random_2, children [WavePlayer_8, WavePlayer_9, WavePlayer_10], properties {bShouldExcludeFromBranchCulling=true, weights [1,1,1]}}
 {id=Random_3, children [WavePlayer_12, WavePlayer_13, WavePlayer_14], properties {bShouldExcludeFromBranchCulling=true, weights [1,1,1]}}
 {id=Random_4, children [WavePlayer_15, WavePlayer_16, WavePlayer_17], properties {bShouldExcludeFromBranchCulling=true, weights [1,1,1]}}
 {
  id=Random_5
  children [WavePlayer_11, WavePlayer_18, WavePlayer_19, WavePlayer_20, WavePlayer_21]
  properties {bShouldExcludeFromBranchCulling=true, weights [1,1,1,1,1]}
 }
 {
  id=SoundClass_0
  children [Attenuation_0]
  properties {soundClassOverride=&Fort_WeaponRanged_Fire_Distant}
 }
 {id=SoundClass_1, children [Modulator_2], properties {soundClassOverride=&Fort_WeaponRanged_Fire_3P_PSM}}
 {
  id=SoundClass_2
  children [Attenuation_5]
  properties {soundClassOverride=&Fort_WeaponRanged_Fire_Distant}
 }
 {id=SoundClass_3, children [Modulator_0], properties {soundClassOverride=&Fort_WeaponRanged_Fire_3P_PSM}}
 {id=WavePlayer_10, properties {soundWaveAssetPtr=/Game/Athena/Sounds/Weapons/Explosions/Clinger_Expl_Close_03}}
 {id=WavePlayer_11, properties {soundWaveAssetPtr=/Game/Athena/Sounds/Weapons/Explosions/EXPL_Debris_Metal_01}}
 {id=WavePlayer_12, properties {soundWaveAssetPtr=/Game/Athena/Sounds/Weapons/Explosions/Clinger_Expl_Med_02}}
 {id=WavePlayer_13, properties {soundWaveAssetPtr=/Game/Athena/Sounds/Weapons/Explosions/Clinger_Expl_Med_03}}
 {id=WavePlayer_14, properties {soundWaveAssetPtr=/Game/Athena/Sounds/Weapons/Explosions/Clinger_Expl_Med_01}}
 {id=WavePlayer_15, properties {soundWaveAssetPtr=/Game/Athena/Sounds/Weapons/Explosions/Grenade_Expl_Dist_01}}
 {id=WavePlayer_16, properties {soundWaveAssetPtr=/Game/Athena/Sounds/Weapons/Explosions/Grenade_Expl_Dist_02}}
 {id=WavePlayer_17, properties {soundWaveAssetPtr=/Game/Athena/Sounds/Weapons/Explosions/Grenade_Expl_Dist_03}}
 {id=WavePlayer_18, properties {soundWaveAssetPtr=/Game/Athena/Sounds/Weapons/Explosions/EXPL_Debris_Metal_02}}
 {id=WavePlayer_19, properties {soundWaveAssetPtr=/Game/Athena/Sounds/Weapons/Explosions/EXPL_Debris_Metal_03}}
 {id=WavePlayer_20, properties {soundWaveAssetPtr=/Game/Athena/Sounds/Weapons/Explosions/EXPL_Debris_Metal_04}}
 {id=WavePlayer_21, properties {soundWaveAssetPtr=/Game/Athena/Sounds/Weapons/Explosions/EXPL_Debris_Metal_05}}
 {id=WavePlayer_8, properties {soundWaveAssetPtr=/Game/Athena/Sounds/Weapons/Explosions/Clinger_Expl_Close_02}}
 {id=WavePlayer_9, properties {soundWaveAssetPtr=/Game/Athena/Sounds/Weapons/Explosions/Clinger_Expl_Close_01}}
]
properties {maxDistance=27500, volumeMultiplier=1}
