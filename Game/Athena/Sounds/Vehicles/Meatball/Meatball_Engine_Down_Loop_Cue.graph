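# Looping exterior engine cue for the Meatball vehicle, a boat-style engine-down layer whose volume and pitch follow the ThrottleOff parameter.
# An engine idle branch, silent without a driver, is mixed with a quieter branch whose volume is also modulated by IsInWater.
# HasDriver crossfades that second branch from a water idle loop to an exterior idle loop.
# Boat engine attenuation, with overrides, handles spatial falloff.
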SoundCue /Game/Athena/Sounds/Vehicles/Meatball/Meatball_Engine_Down_Loop_Cue
root=Mixer_0
nodes [
 {id=Mixer_0, children [ModulatorContinuous_0, ModulatorContinuous_2], properties {inputVolume [1,0.3333]}}
 {
  id=ModulatorContinuous_0
  children [ParamCrossFade_0]
  properties {pitchModulationParams {default=0, minOutput=1.2, parameterName=ThrottleOff}, volumeModulationParams {default=0, parameterName=ThrottleOff}}
 }
 {
  id=ModulatorContinuous_1
  children [ParamCrossFade_1]
  properties {pitchModulationParams {default=0, minOutput=1.2, parameterName=ThrottleOff}, volumeModulationParams {default=0, parameterName=ThrottleOff}}
 }
 {
  id=ModulatorContinuous_2
  children [ModulatorContinuous_1]
  properties {pitchModulationParams {default=0, minOutput=1}, volumeModulationParams {default=0, parameterName=IsInWater}}
 }
 {
  id=ParamCrossFade_0
  children [null,WavePlayer_1]
  properties {
   paramName=HasDriver
   crossFadeInput [
    {fadeInDistanceEnd=0, fadeInDistanceStart=0, fadeOutDistanceEnd=1, fadeOutDistanceStart=0, volume=1}
    {fadeInDistanceEnd=1, fadeInDistanceStart=0, fadeOutDistanceEnd=100, fadeOutDistanceStart=1, volume=1}
   ]
  }
 }
 {
  id=ParamCrossFade_1
  children [WavePlayer_2,WavePlayer_0]
  properties {
   paramName=HasDriver
   crossFadeInput [
    {fadeInDistanceEnd=0, fadeInDistanceStart=0, fadeOutDistanceEnd=1, fadeOutDistanceStart=0, volume=1}
    {fadeInDistanceEnd=1, fadeInDistanceStart=0, fadeOutDistanceEnd=100, fadeOutDistanceStart=1, volume=1}
   ]
  }
 }
 {
  id=WavePlayer_0
  properties {bLooping=true, soundWaveAssetPtr=/Game/Athena/Sounds/Vehicles/Meatball/SoundWaves/Veh_Meatball_Idle_Ext_Loop_01}
 }
 {
  id=WavePlayer_1
  properties {bLooping=true, soundWaveAssetPtr=/Game/Athena/Sounds/Vehicles/Meatball/SoundWaves/Veh_Meatball_Idle_Engine_Loop_01}
 }
 {id=WavePlayer_2, properties {bLooping=true, soundWaveAssetPtr=/Game/Athena/Sounds/Vehicles/Meatball/SoundWaves/Veh_Boat_Idle_Water_Loop_01}}
]
properties {
 attenuationSettings=/Game/Sounds/Attenuation/Fort_Boat_Engine_Close_Detail_Att
 maxDistance=18000
 soundClassObject=&Fort_Vehicles_Exterior
 volumeMultiplier=1.25
 attenuationOverrides {falloffDistance=5000, attenuationShapeExtents {x=500}}
}
