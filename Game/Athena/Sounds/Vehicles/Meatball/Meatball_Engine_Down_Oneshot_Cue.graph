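# One-shot sound cue for the Meatball vehicle's engine winding down, using medium deceleration waves in the exterior vehicle sound class.
# A SpeedThreshold switch gates everything, mixing a random engine decel with an IsInWater switch that adds a random exterior decel and a delayed random water wake.
# Engine and exterior layers ramp up in volume through envelopers and share boat engine attenuation, while the wake uses boat water attenuation.
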SoundCue /Game/Athena/Sounds/Vehicles/Meatball/Meatball_Engine_Down_Oneshot_Cue
root=Switch_1
nodes [
 {id=Attenuation_0, children [Delay_0], properties {attenuationSettings=/Game/Sounds/Attenuation/Fort_Boat_Water_Att}}
 {id=Attenuation_1, children [Enveloper_1], properties {attenuationSettings=/Game/Sounds/Attenuation/Fort_Boat_Engine_Att}}
 {id=Attenuation_2, children [Enveloper_0], properties {attenuationSettings=&Fort_Boat_Engine_Att}}
 {id=Delay_0, children [Random_2], properties {delayMax=1.5, delayMin=1.5}}
 {
  id=Enveloper_0
  children [Random_0]
  properties {pitchCurve {editorCurveData {keys [{time=0}]}}, volumeCurve {editorCurveData {keys [<at 0 0.1>,<at 1 1>]}}}
 }
 {
  id=Enveloper_1
  children [Random_1]
  properties {pitchCurve {editorCurveData {keys [{time=0}]}}, volumeCurve {editorCurveData {keys [<at 0 0.1>,<at 1 1>]}}}
 }
 {id=Mixer_0, children [Attenuation_1, Attenuation_0], properties {inputVolume [1,1]}}
 {id=Mixer_1, children [Attenuation_2, Switch_0], properties {inputVolume [1,1]}}
 {id=Random_0, children [WavePlayer_0,WavePlayer_1], properties {bShouldExcludeFromBranchCulling=true, weights [1,1]}}
 {id=Random_1, children [WavePlayer_2,WavePlayer_3], properties {bShouldExcludeFromBranchCulling=true, weights [1,1]}}
 {id=Random_2, children [WavePlayer_4,WavePlayer_5,WavePlayer_6], properties {bShouldExcludeFromBranchCulling=true, weights [1,1,1]}}
 {id=Switch_0, children [null,null,Mixer_0], properties {intParameterName=IsInWater}}
 {id=Switch_1, children [null,null,Mixer_1], properties {intParameterName=SpeedThreshold}}
 {id=WavePlayer_0, properties {soundWaveAssetPtr=/Game/Athena/Sounds/Vehicles/Meatball/SoundWaves/Veh_Meatball_Med_Decel_Engine_01}}
 {id=WavePlayer_1, properties {soundWaveAssetPtr=/Game/Athena/Sounds/Vehicles/Meatball/SoundWaves/Veh_Meatball_Med_Decel_Engine_02}}
 {id=WavePlayer_2, properties {soundWaveAssetPtr=/Game/Athena/Sounds/Vehicles/Meatball/SoundWaves/Veh_Meatball_Med_Decel_Ext_01}}
 {id=WavePlayer_3, properties {soundWaveAssetPtr=/Game/Athena/Sounds/Vehicles/Meatball/SoundWaves/Veh_Meatball_Med_Decel_Ext_02}}
 {id=WavePlayer_4, properties {soundWaveAssetPtr=/Game/Athena/Sounds/Vehicles/Meatball/SoundWaves/Veh_Meatball_Decel_Water_Wake_02}}
 {id=WavePlayer_5, properties {soundWaveAssetPtr=/Game/Athena/Sounds/Vehicles/Meatball/SoundWaves/Veh_Meatball_Decel_Water_Wake_03}}
 {id=WavePlayer_6, properties {soundWaveAssetPtr=/Game/Athena/Sounds/Vehicles/Meatball/SoundWaves/Veh_Meatball_Decel_Water_Wake_01}}
]
properties {
 maxDistance=18000
 soundClassObject=&Fort_Vehicles_Exterior
 volumeMultiplier=1
 attenuationOverrides {falloffDistance=5000, attenuationShapeExtents {x=500}}
}
