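# Sound cue for the Meatball vehicle's engine shutting off, using boat engine attenuation and the exterior vehicle sound class.
# A Mixer layers an engine turn-off wave, which always plays, with a switched exterior turn-off wave.
# The Switch on the IsInWater int parameter has only its last input filled, so the exterior layer plays only when the vehicle is in water.
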SoundCue /Game/Athena/Sounds/Vehicles/Meatball/Meatball_Engine_Off_Cue
root=Mixer_1
nodes [
 {id=Mixer_1, children [Switch_0,WavePlayer_1], properties {inputVolume [1,1]}}
 {id=Switch_0, children [null,null,WavePlayer_0], properties {intParameterName=IsInWater}}
 {id=WavePlayer_0, properties {soundWaveAssetPtr=/Game/Athena/Sounds/Vehicles/Meatball/SoundWaves/Veh_Meatball_Turn_Off_Ext_01}}
 {id=WavePlayer_1, properties {soundWaveAssetPtr=/Game/Athena/Sounds/Vehicles/Meatball/SoundWaves/Veh_Meatball_Turn_Off_Engine_01}}
]
properties {
 attenuationSettings=/Game/Sounds/Attenuation/Fort_Boat_Engine_Close_Detail_Mono_Att
 maxDistance=18000
 soundClassObject=&Fort_Vehicles_Exterior
 volumeMultiplier=0.65
 attenuationOverrides {falloffDistance=5000, attenuationShapeExtents {x=500}}
}
