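# Sound cue for the Meatball vehicle's engine turn-on, played as an exterior vehicle sound with boat engine attenuation.
# After a fixed half-second delay, it mixes a random engine start wave with an IsInWater switch.
# The switch adds a random exterior turn-on wave only in its last branch, the in-water case, and is otherwise silent.
# Concurrency allows one instance per owner, stopping the oldest.
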SoundCue /Game/Athena/Sounds/Vehicles/Meatball/Meatball_Engine_On_Cue
root=Delay_0
nodes [
 {id=Delay_0, children [Mixer_0], properties {delayMax=0.5, delayMin=0.5}}
 {id=Mixer_0, children [Switch_0,Random_2], properties {inputVolume [1,1]}}
 {id=Random_1, children [WavePlayer_2,WavePlayer_3], properties {bShouldExcludeFromBranchCulling=true, weights [1,1]}}
 {id=Random_2, children [WavePlayer_4,WavePlayer_5], properties {bShouldExcludeFromBranchCulling=true, weights [1,1]}}
 {id=Switch_0, children [null,null,Random_1], properties {intParameterName=IsInWater}}
 {id=WavePlayer_2, properties {soundWaveAssetPtr=/Game/Athena/Sounds/Vehicles/Meatball/SoundWaves/Veh_Meatball_Turn_On_Ext_01}}
 {id=WavePlayer_3, properties {soundWaveAssetPtr=/Game/Athena/Sounds/Vehicles/Meatball/SoundWaves/Veh_Meatball_Turn_On_Ext_02}}
 {id=WavePlayer_4, properties {soundWaveAssetPtr=/Game/Athena/Sounds/Vehicles/Meatball/SoundWaves/Veh_Meatball_Turn_On_Engine_02}}
 {id=WavePlayer_5, properties {soundWaveAssetPtr=/Game/Athena/Sounds/Vehicles/Meatball/SoundWaves/Veh_Meatball_Turn_On_Engine_01}}
]
properties {
 attenuationSettings=/Game/Sounds/Attenuation/Fort_Boat_Engine_Close_Detail_Mono_Att
 bOverrideConcurrency=true
 maxDistance=18000
 soundClassObject=&Fort_Vehicles_Exterior
 volumeMultiplier=1
 attenuationOverrides {falloffDistance=5000, attenuationShapeExtents {x=500}}
 concurrencyOverrides {bLimitToOwner=true, maxCount=1, resolutionRule=StopOldest}
}
