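# Looping engine sound cue for the Meatball vehicle, played with boat engine attenuation in the exterior vehicle sound class.
# It mixes a medium engine loop with a quieter exterior loop whose volume is also scaled by IsInWater.
# On both layers ThrottleOn drives volume, Steer raises pitch, and an IsBoosting crossfade fades the loop out, with no boost sound wired in.
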SoundCue /Game/Athena/Sounds/Vehicles/Meatball/Meatball_Engine_Up_Loop_Cue
root=Mixer_0
nodes [
 {id=Mixer_0, children [ModulatorContinuous_1, ModulatorContinuous_5], properties {inputVolume [1,0.3333]}}
 {
  id=ModulatorContinuous_0
  children [ParamCrossFade_0]
  properties {pitchModulationParams {default=0, minOutput=1, parameterName=ThrottleOn}, volumeModulationParams {default=0, parameterName=ThrottleOn}}
 }
 {
  id=ModulatorContinuous_1
  children [ModulatorContinuous_0]
  properties {pitchModulationParams {default=0, maxOutput=1.2, minOutput=1, parameterName=Steer}, volumeModulationParams {default=0, minOutput=1}}
 }
 {
  id=ModulatorContinuous_3
  children [ParamCrossFade_1]
  properties {pitchModulationParams {default=0, minOutput=1, parameterName=ThrottleOn}, volumeModulationParams {default=0, parameterName=ThrottleOn}}
 }
 {
  id=ModulatorContinuous_4
  children [ModulatorContinuous_3]
  properties {pitchModulationParams {default=0, maxOutput=1.2, minOutput=1, parameterName=Steer}, volumeModulationParams {default=0, minOutput=1}}
 }
 {
  id=ModulatorContinuous_5
  children [ModulatorContinuous_4]
  properties {pitchModulationParams {default=0, minOutput=1}, volumeModulationParams {default=0, parameterName=IsInWater}}
 }
 {
  id=ParamCrossFade_0
  children [WavePlayer_1]
  properties {paramName=IsBoosting, crossFadeInput [{fadeInDistanceEnd=0, fadeInDistanceStart=0, fadeOutDistanceEnd=1, fadeOutDistanceStart=0, volume=1}]}
 }
 {
  id=ParamCrossFade_1
  children [WavePlayer_2,null]
  properties {
   paramName=IsBoosting
   crossFadeInput [
    {fadeInDistanceEnd=0, fadeInDistanceStart=0, fadeOutDistanceEnd=1, fadeOutDistanceStart=0, volume=1}
    {fadeInDistanceEnd=1, fadeInDistanceStart=0, fadeOutDistanceEnd=100, fadeOutDistanceStart=1, volume=1}
   ]
  }
 }
 {
  id=WavePlayer_1
  properties {bLooping=true, soundWaveAssetPtr=/Game/Athena/Sounds/Vehicles/Meatball/SoundWaves/Veh_Meatball_Med_Engine_Loop_01}
 }
 {id=WavePlayer_2, properties {bLooping=true, soundWaveAssetPtr=/Game/Athena/Sounds/Vehicles/Meatball/SoundWaves/Veh_Meatball_Med_Ext_Loop_01}}
]
properties {
 attenuationSettings=/Game/Sounds/Attenuation/Fort_Boat_Engine_Att
 maxDistance=18000
 soundClassObject=&Fort_Vehicles_Exterior
 volumeMultiplier=1.25
 attenuationOverrides {falloffDistance=5000, attenuationShapeExtents {x=500}}
}
