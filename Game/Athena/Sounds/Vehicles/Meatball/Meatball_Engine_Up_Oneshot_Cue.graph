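# One-shot sound cue for the Meatball boat's engine rev-up, played as an exterior vehicle sound with boat engine attenuation.
# A mixer layers the medium engine acceleration wave, shaped by an enveloper that eases its volume down slightly, with a quieter exterior acceleration wave.
# A switch on the IsInWater integer parameter gates the exterior layer, which is audible only when the value is 1.
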SoundCue /Game/Athena/Sounds/Vehicles/Meatball/Meatball_Engine_Up_Oneshot_Cue
root=Mixer_0
nodes [
 {
  id=Enveloper_0
  children [WavePlayer_3]
  properties {pitchCurve {editorCurveData {keys [{time=0}]}}, volumeCurve {editorCurveData {keys [<at 0 1>,<at 3 0.8>]}}}
 }
 {id=Mixer_0, children [Enveloper_0,Switch_0], properties {inputVolume [1,0.3333]}}
 {id=Switch_0, children [null,null,WavePlayer_0], properties {intParameterName=IsInWater}}
 {id=WavePlayer_0, properties {soundWaveAssetPtr=/Game/Athena/Sounds/Vehicles/Meatball/SoundWaves/Veh_Meatball_Med_Accel_Ext_01}}
 {id=WavePlayer_3, properties {soundWaveAssetPtr=/Game/Athena/Sounds/Vehicles/Meatball/SoundWaves/Veh_Meatball_Med_Accel_Engine_01}}
]
properties {
 attenuationSettings=/Game/Sounds/Attenuation/Fort_Boat_Engine_Att
 maxDistance=18000
 soundClassObject=&Fort_Vehicles_Exterior
 volumeMultiplier=1.5
 attenuationOverrides {falloffDistance=5000, attenuationShapeExtents {x=500}}
}
