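# Sound cue for the Meatball vehicle's launcher reload, playing a rocket-ready sound.
# A single wave player feeds the root directly, with no randomization or mixing, in the weapon foley sound class with a slight volume boost.
# It borrows the boat engine attenuation, and a concurrency override caps simultaneous copies per owner, blocking new plays at the limit.
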SoundCue /Game/Athena/Sounds/Vehicles/Meatball/Meatball_Launcher_Reload_Cue
root=WavePlayer_1
nodes [{id=WavePlayer_1, properties {soundWaveAssetPtr=/Game/Athena/Sounds/Vehicles/Meatball/SoundWaves/Veh_Meatball_Rocket_Ready_01}}]
properties {
 attenuationSettings=/Game/Sounds/Attenuation/Fort_Boat_Engine_Att
 bOverrideConcurrency=true
 maxDistance=18000
 soundClassObject=&Fort_WeaponFoley
 volumeMultiplier=1.2
 concurrencyOverrides {bLimitToOwner=true, maxCount=3, resolutionRule=PreventNew}
}
