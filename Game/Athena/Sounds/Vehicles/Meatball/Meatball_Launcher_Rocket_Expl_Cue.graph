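# Sound cue for the Meatball vehicle launcher's rocket explosion, built from Ostrich vehicle launcher explosion waves.
# Mixes close, medium and distant layers, each pairing a random blast with a random tail through a modulator and its own distance attenuation.
# The close layer uses the third-person weapon fire sound class, while the quieter medium and distant layers use the distant-fire class.
# A concurrency override limits simultaneous instances to three.
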SoundCue /Game/Athena/Sounds/Vehicles/Meatball/Meatball_Launcher_Rocket_Expl_Cue
root=Mixer_4
nodes [
 {
  id=Attenuation_0
  children [Modulator_1]
  properties {
   attenuationSettings=/Game/Sounds/Attenuation/Fort_Explosion_Distant_Athena_Att
   attenuationOverrides {falloffDistance=3000, attenuationShapeExtents {x=1000}}
  }
 }
 {
  id=Attenuation_4
  children [Mixer_1]
  properties {attenuationSettings=/Game/Sounds/Attenuation/Fort_Explosion_Close_Athena_Att, attenuationOverrides {falloffDistance=1600}}
 }
 {
  id=Attenuation_5
  children [Modulator_5]
  properties {
   attenuationSettings=/Game/Sounds/Attenuation/Fort_Explosion_Medium_Athena_Att
   attenuationOverrides {falloffDistance=3000, attenuationShapeExtents {x=1000}}
  }
 }
 {id=Mixer_1, children [Random_0,Random_3], properties {inputVolume [1,1]}}
 {id=Mixer_2, children [Random_4,Random_1], properties {inputVolume [1,1]}}
 {id=Mixer_3, children [Random_5,Random_2], properties {inputVolume [1,1]}}
 {id=Mixer_4, children [SoundClass_3,SoundClass_2,SoundClass_0], properties {inputVolume [1,0.75,0.75]}}
 {id=Modulator_0, children [Attenuation_4]}
 {id=Modulator_1, children [Mixer_3], properties {pitchMax=1.1, pitchMin=0.9}}
 {id=Modulator_5, children [Mixer_2], properties {pitchMax=1.1, pitchMin=0.9}}
 {
  id=Random_0
  children [WavePlayer_15, WavePlayer_16, WavePlayer_17, WavePlayer_27, WavePlayer_28]
  properties {bShouldExcludeFromBranchCulling=true, weights [1,1,1,1,1]}
 }
 {
  id=Random_1
  children [WavePlayer_0, WavePlayer_1, WavePlayer_4, WavePlayer_12]
  properties {bShouldExcludeFromBranchCulling=true, weights [1,1,1,1]}
 }
 {
  id=Random_2
  children [WavePlayer_29, WavePlayer_30, WavePlayer_31, WavePlayer_32]
  properties {bShouldExcludeFromBranchCulling=true, weights [1,1,1,1]}
 }
 {
  id=Random_3
  children [WavePlayer_2,WavePlayer_3,WavePlayer_5,WavePlayer_6]
  properties {bShouldExcludeFromBranchCulling=true, weights [1,1,1,1]}
 }
 {
  id=Random_4
  children [WavePlayer_9, WavePlayer_10, WavePlayer_11, WavePlayer_18, WavePlayer_19, WavePlayer_20]
  properties {bShouldExcludeFromBranchCulling=true, weights [1,1,1,1,1,1]}
 }
 {
  id=Random_5
  children [WavePlayer_21, WavePlayer_22, WavePlayer_23, WavePlayer_24, WavePlayer_25, WavePlayer_26]
  properties {bShouldExcludeFromBranchCulling=true, weights [1,1,1,1,1,1]}
 }
 {
  id=SoundClass_0
  children [Attenuation_0]
  properties {soundClassOverride=&Fort_WeaponRanged_Fire_Distant}
 }
 {
  id=SoundClass_2
  children [Attenuation_5]
  properties {soundClassOverride=&Fort_WeaponRanged_Fire_Distant}
 }
 {id=SoundClass_3, children [Modulator_0], properties {soundClassOverride=&Fort_WeaponRanged_Fire_3P_PSM}}
 {id=WavePlayer_0, properties {soundWaveAssetPtr=/Game/Athena/Sounds/Vehicles/Ostrich/Veh_Ostrich_Launcher_Expl_Tail_Med_01}}
 {id=WavePlayer_1, properties {soundWaveAssetPtr=/Game/Athena/Sounds/Vehicles/Ostrich/Veh_Ostrich_Launcher_Expl_Tail_Med_02}}
 {id=WavePlayer_10, properties {soundWaveAssetPtr=/Game/Athena/Sounds/Vehicles/Ostrich/Veh_Ostrich_Launcher_Expl_Med_02}}
 {id=WavePlayer_11, properties {soundWaveAssetPtr=/Game/Athena/Sounds/Vehicles/Ostrich/Veh_Ostrich_Launcher_Expl_Med_03}}
 {id=WavePlayer_12, properties {soundWaveAssetPtr=/Game/Athena/Sounds/Vehicles/Ostrich/Veh_Ostrich_Launcher_Expl_Tail_Med_04}}
 {id=WavePlayer_15, properties {soundWaveAssetPtr=/Game/Athena/Sounds/Vehicles/Ostrich/Veh_Ostrich_Launcher_Expl_Tail_Close_01}}
 {id=WavePlayer_16, properties {soundWaveAssetPtr=/Game/Athena/Sounds/Vehicles/Ostrich/Veh_Ostrich_Launcher_Expl_Tail_Close_02}}
 {id=WavePlayer_17, properties {soundWaveAssetPtr=/Game/Athena/Sounds/Vehicles/Ostrich/Veh_Ostrich_Launcher_Expl_Tail_Close_03}}
 {id=WavePlayer_18, properties {soundWaveAssetPtr=/Game/Athena/Sounds/Vehicles/Ostrich/Veh_Ostrich_Launcher_Expl_Med_04}}
 {id=WavePlayer_19, properties {soundWaveAssetPtr=/Game/Athena/Sounds/Vehicles/Ostrich/Veh_Ostrich_Launcher_Expl_Med_05}}
 {id=WavePlayer_2, properties {soundWaveAssetPtr=/Game/Athena/Sounds/Vehicles/Ostrich/Veh_Ostrich_Launcher_Expl_Close_01}}
 {id=WavePlayer_20, properties {soundWaveAssetPtr=/Game/Athena/Sounds/Vehicles/Ostrich/Veh_Ostrich_Launcher_Expl_Med_06}}
 {id=WavePlayer_21, properties {soundWaveAssetPtr=/Game/Athena/Sounds/Vehicles/Ostrich/Veh_Ostrich_Launcher_Expl_Dist_01}}
 {id=WavePlayer_22, properties {soundWaveAssetPtr=/Game/Athena/Sounds/Vehicles/Ostrich/Veh_Ostrich_Launcher_Expl_Dist_02}}
 {id=WavePlayer_23, properties {soundWaveAssetPtr=/Game/Athena/Sounds/Vehicles/Ostrich/Veh_Ostrich_Launcher_Expl_Dist_03}}
 {id=WavePlayer_24, properties {soundWaveAssetPtr=/Game/Athena/Sounds/Vehicles/Ostrich/Veh_Ostrich_Launcher_Expl_Dist_04}}
 {id=WavePlayer_25, properties {soundWaveAssetPtr=/Game/Athena/Sounds/Vehicles/Ostrich/Veh_Ostrich_Launcher_Expl_Dist_05}}
 {id=WavePlayer_26, properties {soundWaveAssetPtr=/Game/Athena/Sounds/Vehicles/Ostrich/Veh_Ostrich_Launcher_Expl_Dist_06}}
 {id=WavePlayer_27, properties {soundWaveAssetPtr=/Game/Athena/Sounds/Vehicles/Ostrich/Veh_Ostrich_Launcher_Expl_Tail_Close_04}}
 {id=WavePlayer_28, properties {soundWaveAssetPtr=/Game/Athena/Sounds/Vehicles/Ostrich/Veh_Ostrich_Launcher_Expl_Tail_Close_05}}
 {id=WavePlayer_29, properties {soundWaveAssetPtr=/Game/Athena/Sounds/Vehicles/Ostrich/Veh_Ostrich_Launcher_Expl_Tail_Dist_01}}
 {id=WavePlayer_3, properties {soundWaveAssetPtr=/Game/Athena/Sounds/Vehicles/Ostrich/Veh_Ostrich_Launcher_Expl_Close_02}}
 {id=WavePlayer_30, properties {soundWaveAssetPtr=/Game/Athena/Sounds/Vehicles/Ostrich/Veh_Ostrich_Launcher_Expl_Tail_Dist_02}}
 {id=WavePlayer_31, properties {soundWaveAssetPtr=/Game/Athena/Sounds/Vehicles/Ostrich/Veh_Ostrich_Launcher_Expl_Tail_Dist_03}}
 {id=WavePlayer_32, properties {soundWaveAssetPtr=/Game/Athena/Sounds/Vehicles/Ostrich/Veh_Ostrich_Launcher_Expl_Tail_Dist_04}}
 {id=WavePlayer_4, properties {soundWaveAssetPtr=/Game/Athena/Sounds/Vehicles/Ostrich/Veh_Ostrich_Launcher_Expl_Tail_Med_03}}
 {id=WavePlayer_5, properties {soundWaveAssetPtr=/Game/Athena/Sounds/Vehicles/Ostrich/Veh_Ostrich_Launcher_Expl_Close_03}}
 {id=WavePlayer_6, properties {soundWaveAssetPtr=/Game/Athena/Sounds/Vehicles/Ostrich/Veh_Ostrich_Launcher_Expl_Close_04}}
 {id=WavePlayer_9, properties {soundWaveAssetPtr=/Game/Athena/Sounds/Vehicles/Ostrich/Veh_Ostrich_Launcher_Expl_Med_01}}
]
properties {
 bOverrideConcurrency=true
 maxDistance=27500
 soundClassObject=&Fort_WeaponRanged_Fire_3P_PSM
 volumeMultiplier=0.8
 concurrencyOverrides {maxCount=3}
 concurrencySet [/Game/Sounds/Concurrency_Rules/Weapon_Explosion_Conc]
}
