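# Looping in-flight sound cue for the Meatball vehicle launcher's rocket.
# A looping fly wave feeds a pitch-varying Modulator, then a SoundClass node that overrides the cue's third-person fire class with the distant ranged-fire class.
# A smoothed, intensified Doppler node at the root bends pitch with relative motion.
# The cue uses missile in-air attenuation and its own per-owner concurrency limit.
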SoundCue /Game/Athena/Sounds/Vehicles/Meatball/Meatball_Launcher_Rocket_InAir_Cue
root=Doppler_1
nodes [
 {id=Doppler_1, children [SoundClass_6], properties {bUseSmoothing=true, dopplerIntensity=3}}
 {id=Modulator_7, children [WavePlayer_0], properties {pitchMax=1.1}}
 {
  id=SoundClass_6
  children [Modulator_7]
  properties {soundClassOverride=&Fort_WeaponRanged_Fire_Distant}
 }
 {
  id=WavePlayer_0
  properties {bLooping=true, soundWaveAssetPtr=/Game/Athena/Sounds/Vehicles/Meatball/SoundWaves/Veh_Meatball_Rocket_Fly_Loop_01}
 }
]
properties {
 attenuationSettings=/Game/Sounds/Attenuation/Fort_Weapon_Projectile_MEatball_Missile_InAir
 bOverrideConcurrency=true
 maxDistance=10512
 soundClassObject=&Fort_WeaponRanged_Fire_3P_PSM
 volumeMultiplier=0.4
 attenuationOverrides {falloffDistance=5000, attenuationShapeExtents {x=1000}}
 concurrencyOverrides {bLimitToOwner=true, maxCount=3, volumeScale=0.5}
}
