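# Looping sound cue for the exterior scrape of the Meatball vehicle, built on a dry slide wave.
# A looping wave player feeds a continuous modulator that drives both pitch and volume from a Speed parameter.
# Pitch varies in a narrow band around normal as Speed changes.
# It plays in the exterior vehicle sound class and references a close-detail boat engine attenuation asset.
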SoundCue /Game/Athena/Sounds/Vehicles/Meatball/Meatball_Scrape_Loop_Cue
root=ModulatorContinuous_0
nodes [
 {
  id=ModulatorContinuous_0
  children [WavePlayer_2]
  properties {
   pitchModulationParams {default=0, maxOutput=1.1, minOutput=0.8, parameterName=Speed}
   volumeModulationParams {default=0, parameterName=Speed}
  }
 }
 {
  id=WavePlayer_2
  properties {bLooping=true, soundWaveAssetPtr=/Game/Athena/Sounds/Vehicles/Meatball/SoundWaves/Veh_Meatball_Dry_Slide_Loop_01}
 }
]
properties {
 attenuationSettings=/Game/Sounds/Attenuation/Fort_Boat_Engine_Close_Detail_Att
 maxDistance=18000
 soundClassObject=&Fort_Vehicles_Exterior
 volumeMultiplier=0.8
 attenuationOverrides {falloffDistance=5000, attenuationShapeExtents {x=500}}
}
