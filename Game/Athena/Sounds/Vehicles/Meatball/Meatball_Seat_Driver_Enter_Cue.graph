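# Sound cue for entering the driver seat of the Meatball vehicle in Athena.
# A single wave player feeds a Modulator at default settings, which adds slight random pitch and volume variation on each play.
# The cue plays at reduced volume in the vehicle interior sound class and borrows the golf cart close-movement attenuation for distance falloff.
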SoundCue /Game/Athena/Sounds/Vehicles/Meatball/Meatball_Seat_Driver_Enter_Cue
root=Modulator_0
nodes [
 {id=Modulator_0, children [WavePlayer_0]}
 {id=WavePlayer_0, properties {soundWaveAssetPtr=/Game/Athena/Sounds/Vehicles/Meatball/SoundWaves/Veh_Meatball_Enter_Driver_01}}
]
properties {
 attenuationSettings=/Game/Sounds/Attenuation/Fort_GolfCart_Movement_Close_Att
 maxDistance=5300
 soundClassObject=&Fort_Vehicles_Interior
 volumeMultiplier=0.5
}
