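# Sound cue for a passenger entering a seat in the Meatball vehicle.
# A random node picks between two equally weighted enter-passenger waves, excluded from branch culling, and feeds a modulator at the root for variation.
# It plays at reduced volume in the vehicle interior sound class, borrowing a close-range golf cart movement attenuation.
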SoundCue /Game/Athena/Sounds/Vehicles/Meatball/Meatball_Seat_Passenger_Enter_Cue
root=Modulator_0
nodes [
 {id=Modulator_0, children [Random_0]}
 {id=Random_0, children [WavePlayer_1,WavePlayer_2], properties {bShouldExcludeFromBranchCulling=true, weights [1,1]}}
 {id=WavePlayer_1, properties {soundWaveAssetPtr=/Game/Athena/Sounds/Vehicles/Meatball/SoundWaves/Veh_Meatball_Enter_Passenger_02}}
 {id=WavePlayer_2, properties {soundWaveAssetPtr=/Game/Athena/Sounds/Vehicles/Meatball/SoundWaves/Veh_Meatball_Enter_Passenger_01}}
]
properties {
 attenuationSettings=/Game/Sounds/Attenuation/Fort_GolfCart_Movement_Close_Att
 maxDistance=5300
 soundClassObject=&Fort_Vehicles_Interior
 volumeMultiplier=0.6
}
