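# Looping exterior speed sound cue for the Meatball vehicle, using boat water attenuation.
# An IsInWater parameter crossfades between a skydive wind loop when out of water and a fast wake loop when in water.
# A Speed parameter drives volume and pitch on both loops.
# The wind branch adds a second modulator that scales volume by IsInAir, boosting it up to double and defaulting to silent.
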SoundCue /Game/Athena/Sounds/Vehicles/Meatball/Meatball_Wake_Speed_Cue
root=ParamCrossFade_1
nodes [
 {
  id=ModulatorContinuous_0
  children [WavePlayer_1]
  properties {pitchModulationParams {maxOutput=1.2, minOutput=0.8, parameterName=Speed}, volumeModulationParams {default=0, parameterName=Speed}}
 }
 {
  id=ModulatorContinuous_1
  children [ModulatorContinuous_0]
  properties {pitchModulationParams {default=0, minOutput=1}, volumeModulationParams {default=0, maxOutput=2, parameterName=IsInAir}}
 }
 {
  id=ModulatorContinuous_3
  children [WavePlayer_0]
  properties {pitchModulationParams {maxOutput=1.2, minOutput=0.8, parameterName=Speed}, volumeModulationParams {default=0, parameterName=Speed}}
 }
 {
  id=ParamCrossFade_1
  children [ModulatorContinuous_1, ModulatorContinuous_3]
  properties {
   paramName=IsInWater
   crossFadeInput [
    {fadeInDistanceEnd=0, fadeInDistanceStart=0, fadeOutDistanceEnd=1, fadeOutDistanceStart=0, volume=1}
    {fadeInDistanceEnd=1, fadeInDistanceStart=0, fadeOutDistanceEnd=100, fadeOutDistanceStart=1, volume=1}
   ]
  }
 }
 {
  id=WavePlayer_0
  properties {bLooping=true, soundWaveAssetPtr=/Game/Athena/Sounds/Vehicles/Meatball/SoundWaves/Veh_Meatball_Wake_Fast_Loop_01}
 }
 {id=WavePlayer_1, properties {bLooping=true, soundWaveAssetPtr=/Game/Athena/Sounds/Wind_Falling/AMB_Wind_Skydive_Vertical_Loop_01}}
]
properties {
 attenuationSettings=/Game/Sounds/Attenuation/Fort_Boat_Water_Att
 maxDistance=8000
 soundClassObject=&Fort_Vehicles_Exterior
 volumeMultiplier=1
 attenuationOverrides {falloffDistance=5000, attenuationShapeExtents {x=500}}
}
