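# Sound cue for the looping wake of the Meatball vehicle while turning, played as exterior vehicle audio with boat-water attenuation.
# A Steer parameter crossfades from a calm wake loop to a choppy one.
# The same parameter drives a continuous modulator that scales volume, silent by default.
# An outer IsInWater crossfade fades the whole sound in when the vehicle is in water.
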SoundCue /Game/Athena/Sounds/Vehicles/Meatball/Meatball_Wake_Turn_Cue
root=ParamCrossFade_1
nodes [
 {
  id=ModulatorContinuous_3
  children [ParamCrossFade_0]
  properties {pitchModulationParams {minOutput=1}, volumeModulationParams {default=0, parameterName=Steer}}
 }
 {
  id=ParamCrossFade_0
  children [WavePlayer_3,WavePlayer_4]
  properties {
   paramName=Steer
   crossFadeInput [
    {fadeInDistanceEnd=0, fadeInDistanceStart=0, fadeOutDistanceEnd=1, fadeOutDistanceStart=0, volume=1}
    {fadeInDistanceEnd=1, fadeInDistanceStart=0, fadeOutDistanceEnd=100, fadeOutDistanceStart=1, volume=1}
   ]
  }
 }
 {
  id=ParamCrossFade_1
  children [ModulatorContinuous_3]
  properties {paramName=IsInWater, crossFadeInput [{fadeInDistanceEnd=1, fadeInDistanceStart=0, fadeOutDistanceEnd=100, fadeOutDistanceStart=1, volume=1}]}
 }
 {
  id=WavePlayer_3
  properties {bLooping=true, soundWaveAssetPtr=/Game/Athena/Sounds/Vehicles/Meatball/SoundWaves/Veh_Meatball_Wake_Calm_Loop_01}
 }
 {
  id=WavePlayer_4
  properties {bLooping=true, soundWaveAssetPtr=/Game/Athena/Sounds/Vehicles/Meatball/SoundWaves/Veh_Meatball_Wake_Choppy_Loop_01}
 }
]
properties {
 attenuationSettings=/Game/Sounds/Attenuation/Fort_Boat_Water_Att
 maxDistance=8000
 soundClassObject=&Fort_Vehicles_Exterior
 volumeMultiplier=3
 attenuationOverrides {falloffDistance=5000, attenuationShapeExtents {x=500}}
}
