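# Sound cue named for the Meatball vehicle entering water, set up as an exterior vehicle sound.
# Its node graph is empty with no root, so the cue as shown has no sound source to play.
# It still carries half volume, the vehicle exterior sound class, and a close-detail boat engine attenuation asset with a maximum distance.
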SoundCue /Game/Athena/Sounds/Vehicles/Meatball/Meatball_Water_Enter_Cue
root=null
nodes []
properties {
 attenuationSettings=/Game/Sounds/Attenuation/Fort_Boat_Engine_Close_Detail_Att
 maxDistance=5300
 soundClassObject=&Fort_Vehicles_Exterior
 volumeMultiplier=0.5
}
