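# Sound cue for the out-of-fuel sound of the Meatball vehicle.
# A single wave player feeds the root directly, routed to the interior vehicle sound class with prop vehicle attenuation.
# A concurrency override limits it to one instance per owner and prevents new plays while one is already sounding.
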SoundCue /Game/Athena/Sounds/Vehicles/Meatball/Veh_Meatball_OutOfFuel_Cue
root=WavePlayer_1
nodes [{id=WavePlayer_1, properties {soundWaveAssetPtr=&/Veh_Meatball_OutOfFuel_01}}]
properties {
 attenuationSettings=/Game/Sounds/Attenuation/Fort_Prop_Vehicles_Att
 bOverrideConcurrency=true
 maxDistance=3000
 soundClassObject=&Fort_Vehicles_Interior
 volumeMultiplier=1
 concurrencyOverrides {bLimitToOwner=true, maxCount=1, resolutionRule=PreventNew}
}
