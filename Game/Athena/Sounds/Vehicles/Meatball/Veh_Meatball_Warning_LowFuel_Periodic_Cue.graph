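# Sound cue for the periodic low-fuel warning on the Meatball vehicle.
# A single wave player feeds the output directly, with no looping or randomization in the cue, and plays in the vehicles sound class at slightly reduced volume.
# It uses the shared player super-close attenuation asset and overrides concurrency so new instances are blocked once the limit is reached.
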SoundCue /Game/Athena/Sounds/Vehicles/Meatball/Veh_Meatball_Warning_LowFuel_Periodic_Cue
root=WavePlayer_0
nodes [{id=WavePlayer_0, properties {soundWaveAssetPtr=&/Veh_Meatball_Warning_LowFuel_Periodic_01}}]
properties {
 attenuationSettings=/Game/Sounds/Attenuation/Fort_Player_SuperClose_Att
 bOverrideConcurrency=true
 maxDistance=800
 soundClassObject=&Fort_Vehicles
 volumeMultiplier=0.8
 attenuationOverrides {distanceAlgorithm=NaturalSound, falloffDistance=4000, attenuationShapeExtents {x=1000}}
 concurrencyOverrides {resolutionRule=PreventNew}
}
